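# Unlit glass material for the KneeLens gadget emote prop, usable on skeletal meshes, that fakes depth inside the lens with an interior cubemap.
# The cubemap's blue and green mask channels get separate tints and sine-driven emissive pulses.
# MaterialCurveA scales that glow and gates the pulse through a sign test.
# The glow is added over the colour texture, which is slightly darkened by an adjustable sphere mask.
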Material /BRCosmetics/Animation/Props/Gadgets_KneeLens/Materials/M_Gadget_KneeLens_Glass
graph {
 properties {
  bAutomaticallySetUsageInEditor=false
  bUsedWithMeshDeformer=true
  bUsedWithSkeletalMesh=true
  shadingModel=MSM_Unlit
  usedShadingModels=Unlit
 }
 nodes [
  n1 = TextureCoordinate()
  n2 = ComponentMask(None=VectorParameter() {parameterName=Scale, defaultValue=<rgba 0.184 0.208 0 1>}.RGB) {g=true, r=true}
  n3 = TextureSample(UVs=MaterialFunctionCall!(UVs=Add(A=@n1, B=@n2), Tiling=VectorParameter() {parameterName=Tiling, defaultValue=<rgba 5.04 4.2784 0 1>}.RGB, Randomize Rotation) {materialFunction=/Engine/Functions/Engine_MaterialFunctions01/Cubemaps/InteriorCubemap}.UVW) {samplerType=SAMPLERTYPE_Masks, texture=/BRCosmetics/Animation/Props/Gadgets_KneeLens/Textures/T_KneeLens_Emote_Prop_Cubemap_Uniformed}
  n4 = ScalarParameter() {defaultValue=0.7, parameterName=TimeOffsetC}
  n5 = Add(
   A=ScalarParameter() {defaultValue=0.3, parameterName=Emissivec}
   B=Multiply(
    A=MaterialFunctionCall(Speed=@n4) {materialFunction=&DebugTimeSine}.Result
    B=ScalarParameter() {defaultValue=2.5, parameterName=TimeScaleC}
   ) {constB=1.1}
  ) {constB=0.73}
  n6 = ScalarParameter() {defaultValue=1, parameterName=MaterialCurveA}
  n7 = OneMinus(None=Abs(None=Sign(None=@n6)))
  n8 = Multiply(
   A=Multiply(A=@n3.B, B=LinearInterpolate(A=@n5, Alpha=@n7) {constA=1})
   B=VectorParameter() {parameterName=ColorC, defaultValue=<rgba 1 0.9214 0 1>}.RGB
  )
  n9 = ScalarParameter() {defaultValue=0.6, parameterName=TimeOffsetB}
  n10 = Add(
   A=ScalarParameter() {defaultValue=0.2, parameterName=EmissiveB}
   B=Multiply(
    A=MaterialFunctionCall(Speed=@n9) {materialFunction=&DebugTimeSine}.Result
    B=ScalarParameter() {defaultValue=4, parameterName=TimeScaleB}
   ) {constB=1.1}
  )
  n11 = Multiply(
   A=VectorParameter() {parameterName=ColorB, defaultValue=<rgba 0.0894 0.4125 1 1>}.RGB
   B=Multiply(A=@n3.G, B=LinearInterpolate(A=@n10, Alpha=@n7) {constA=1})
  )
  n12 = VectorParameter() {parameterName=CenterDarkening, defaultValue=<rgba 0.508 0.612 0.1878 1>}
  n13 = Power(Base=SphereMask(A=TextureCoordinate(), B=ComponentMask(None=@n12.RGB) {g=true, r=true}, Radius=@n12.B) {attenuationRadius=0.1, hardnessPercent=0}) {constExponent=1}
  Add(
   A=Multiply(A=Add(A=@n8, B=@n11), B=ScalarParameter() {defaultValue=1, parameterName=MaterialCurveA})
   B=Subtract(
    A=TextureSample() {texture=/BRCosmetics/Animation/Props/Gadgets_KneeLens/Textures/T_KneeLens_Emote_Prop_CL}.RGB
    B=Multiply(A=@n13) {constB=0.05}
   )
  )
  Constant() {r=1}
 ]
}
